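# Sound cue for the drop sound of the Chicago item in the Ashton limited-time mode.
# A single wave player plays the Ashton_Chicago_Drop wave with no randomization or mixing.
# It uses the shared weapon-drop attenuation, the loot-collect sound class and the Weapon_Drop_StopOldest concurrency set.
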SoundCue /Game/Athena/Sounds/LTMs/Ashton/Chicago/Ashton_Chicago_Drop_Cue
root=WavePlayer_0
nodes [{id=WavePlayer_0, properties {soundWaveAssetPtr=&/Ashton_Chicago_Drop}}]
properties {
 attenuationSettings=&Fort_Weapon_Drop_Att
 maxDistance=3000
 soundClassObject=&Fort_Loot_Collect
 concurrencyOverrides {resolutionRule=PreventNew}
 concurrencySet [&Weapon_Drop_StopOldest]
}
