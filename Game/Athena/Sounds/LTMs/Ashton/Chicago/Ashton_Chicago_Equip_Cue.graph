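# Sound cue for the Ashton Chicago limited-time-mode equip sound, played as weapon foley.
# A single wave player feeds the Ashton_Chicago_Equip wave straight to the output, with no randomization or mixing.
# It uses the weapon reload attenuation and the weapon foley sound class, overrides concurrency, and has debug enabled.
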SoundCue /Game/Athena/Sounds/LTMs/Ashton/Chicago/Ashton_Chicago_Equip_Cue
root=WavePlayer_1
nodes [{id=WavePlayer_1, properties {soundWaveAssetPtr=&/Ashton_Chicago_Equip}}]
properties {
 attenuationSettings=&Fort_Weapon_Reload_Att
 bDebug=true
 bOverrideConcurrency=true
 maxDistance=2250
 soundClassObject=&Fort_WeaponFoley
 volumeMultiplier=1
}
