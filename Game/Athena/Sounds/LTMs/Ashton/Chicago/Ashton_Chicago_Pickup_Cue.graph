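# Sound cue for picking up the Ashton Chicago item in a Fortnite limited-time mode.
# A mixer layers the dedicated pickup wave, delayed by a fixed fraction of a second, with an immediate, much quieter loot reward music stinger.
# It borrows weapon reload attenuation and a third-person ranged-weapon fire sound class, and references a shared pickup concurrency set to limit overlapping plays.
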SoundCue /Game/Athena/Sounds/LTMs/Ashton/Chicago/Ashton_Chicago_Pickup_Cue
root=Mixer_0
nodes [
 {id=Delay_0, children [WavePlayer_4], properties {delayMax=0.3, delayMin=0.3}}
 {id=Mixer_0, children [Delay_0,WavePlayer_0], properties {inputVolume [1,0.2]}}
 {id=WavePlayer_0, properties {soundWaveAssetPtr=/Game/Sounds/Fort_Music/Stingers/loot_reward_02_mix}}
 {id=WavePlayer_4, properties {soundWaveAssetPtr=&/Ashton_Chicago_Pickup}}
]
properties {
 attenuationSettings=&Fort_Weapon_Reload_Att
 maxDistance=2250
 soundClassObject=&Fort_WeaponRanged_Fire_3P_PSM
 volumeMultiplier=1
 concurrencyOverrides {maxCount=1, resolutionRule=PreventNew}
 concurrencySet [&PickupMaximum_Oldest]
}
