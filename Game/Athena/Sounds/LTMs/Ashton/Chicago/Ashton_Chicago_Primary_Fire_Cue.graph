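# Sound cue for the Ashton Chicago LTM primary fire, a throw sound built from start and release layers.
# A mixer plays one of four random throw-start waves at once, plus one of four random release waves after a fixed short delay at reduced volume.
# It uses close projectile attenuation and the third-person ranged fire sound class, with a concurrency override that prevents new instances at the limit.
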SoundCue /Game/Athena/Sounds/LTMs/Ashton/Chicago/Ashton_Chicago_Primary_Fire_Cue
root=Mixer_0
nodes [
 {id=Delay_0, children [Random_3], properties {delayMax=0.4, delayMin=0.4}}
 {id=Mixer_0, children [Random_1,Delay_0], properties {inputVolume [1,0.75]}}
 {id=Random_1, children [WavePlayer_2,WavePlayer_3,WavePlayer_4,WavePlayer_5], properties {weights [1,1,1,1]}}
 {id=Random_3, children [WavePlayer_6,WavePlayer_7,WavePlayer_8,WavePlayer_9], properties {weights [1,1,1,1]}}
 {id=WavePlayer_2, properties {soundWaveAssetPtr=&/Ashton_Chicago_Throw_Start_01}}
 {id=WavePlayer_3, properties {soundWaveAssetPtr=&/Ashton_Chicago_Throw_Start_02}}
 {id=WavePlayer_4, properties {soundWaveAssetPtr=&/Ashton_Chicago_Throw_Start_03}}
 {id=WavePlayer_5, properties {soundWaveAssetPtr=&/Ashton_Chicago_Throw_Start_04}}
 {id=WavePlayer_6, properties {soundWaveAssetPtr=&/Ashton_Chicago_Throw_Release_01}}
 {id=WavePlayer_7, properties {soundWaveAssetPtr=&/Ashton_Chicago_Throw_Release_02}}
 {id=WavePlayer_8, properties {soundWaveAssetPtr=&/Ashton_Chicago_Throw_Release_03}}
 {id=WavePlayer_9, properties {soundWaveAssetPtr=&/Ashton_Chicago_Throw_Release_04}}
]
properties {
 attenuationSettings=&Fort_Weapon_Projectile_Close_Athena_Att
 bOverrideConcurrency=true
 maxDistance=4950
 soundClassObject=&Fort_WeaponRanged_Fire_3P_PSM
 volumeMultiplier=1
 attenuationOverrides {falloffDistance=5000, attenuationShapeExtents {x=1000}}
 concurrencyOverrides {resolutionRule=PreventNew}
}
