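# Sound cue for the enemy impact of the Ashton Chicago primary throw, a limited-time-mode weapon in Athena.
# A random node picks one of six throw impact waves with equal weights.
# It plays through medium mono projectile attenuation in the enemy-damaged sound class.
# A concurrency override prevents new instances when the limit is reached.
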SoundCue /Game/Athena/Sounds/LTMs/Ashton/Chicago/Ashton_Chicago_Primary_Impt_Enemy_Cue
root=Random_1
nodes [
 {id=Random_1, children [WavePlayer_2,WavePlayer_3,WavePlayer_4,WavePlayer_5,WavePlayer_6,WavePlayer_7], properties {weights [1,1,1,1,1,1]}}
 {id=WavePlayer_2, properties {soundWaveAssetPtr=&/Ashton_Chicago_Throw_Enemy_Impact_01}}
 {id=WavePlayer_3, properties {soundWaveAssetPtr=&/Ashton_Chicago_Throw_Enemy_Impact_02}}
 {id=WavePlayer_4, properties {soundWaveAssetPtr=&/Ashton_Chicago_Throw_Enemy_Impact_03}}
 {id=WavePlayer_5, properties {soundWaveAssetPtr=&/Ashton_Chicago_Throw_Enemy_Impact_04}}
 {id=WavePlayer_6, properties {soundWaveAssetPtr=&/Ashton_Chicago_Throw_Enemy_Impact_05}}
 {id=WavePlayer_7, properties {soundWaveAssetPtr=&/Ashton_Chicago_Throw_Enemy_Impact_06}}
]
properties {
 attenuationSettings=&Fort_Weapon_Projectile_Medium_Mono_Athena_Att
 bOverrideConcurrency=true
 maxDistance=10579.2871
 soundClassObject=/Game/Sounds/FortSoundClassesAndModes/Fort_Enemy_Damaged_PSM
 volumeMultiplier=1
 attenuationOverrides {falloffDistance=5000, attenuationShapeExtents {x=1000}}
 concurrencyOverrides {resolutionRule=PreventNew}
}
