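# This sound cue plays the environment impact sound for the primary Chicago throw weapon in the Ashton limited-time mode.
# A Random node picks one of six equally weighted throw-impact wave variations.
# It routes to the impacts sound class, uses natural-sound distance attenuation overrides, and limits concurrent instances through a concurrency override.
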SoundCue /Game/Athena/Sounds/LTMs/Ashton/Chicago/Ashton_Chicago_Primary_Impt_Env_Cue
root=Random_0
nodes [
 {
  id=Random_0
  children [WavePlayer_0, WavePlayer_1, WavePlayer_10, WavePlayer_11, WavePlayer_12, WavePlayer_13]
  properties {weights [1,1,1,1,1,1]}
 }
 {id=WavePlayer_0, properties {soundWaveAssetPtr=&/Ashton_Chicago_Throw_Env_Impact_01}}
 {id=WavePlayer_1, properties {soundWaveAssetPtr=&/Ashton_Chicago_Throw_Env_Impact_02}}
 {id=WavePlayer_10, properties {soundWaveAssetPtr=&/Ashton_Chicago_Throw_Env_Impact_03}}
 {id=WavePlayer_11, properties {soundWaveAssetPtr=&/Ashton_Chicago_Throw_Env_Impact_04}}
 {id=WavePlayer_12, properties {soundWaveAssetPtr=&/Ashton_Chicago_Throw_Env_Impact_05}}
 {id=WavePlayer_13, properties {soundWaveAssetPtr=&/Ashton_Chicago_Throw_Env_Impact_06}}
]
properties {
 attenuationSettings=&Fort_Weapon_Projectile_Medium_Mono_Athena_Att
 bOverrideAttenuation=true
 bOverrideConcurrency=true
 maxDistance=7000
 soundClassObject=&Fort_Impacts
 volumeMultiplier=1
 attenuationOverrides {distanceAlgorithm=NaturalSound, falloffDistance=6000, attenuationShapeExtents {x=1000}}
 concurrencyOverrides {maxCount=4}
}
