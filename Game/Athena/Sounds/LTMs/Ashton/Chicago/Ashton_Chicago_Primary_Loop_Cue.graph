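# Sound cue for the looping primary throw sound of the Ashton Chicago item in an Athena limited-time mode.
# A random node picks one of two looping throw waves with equal weight and feeds it through an attenuation node using close weapon-projectile settings.
# The cue plays in the ranged weapon sound class and overrides concurrency so that new instances are prevented once the limit is reached.
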SoundCue /Game/Athena/Sounds/LTMs/Ashton/Chicago/Ashton_Chicago_Primary_Loop_Cue
root=Attenuation_0
nodes [
 {id=Attenuation_0, children [Random_0], properties {attenuationSettings=&Fort_Weapon_Projectile_Close_Athena_Att}}
 {id=Random_0, children [WavePlayer_0,WavePlayer_1], properties {weights [1,1]}}
 {id=WavePlayer_0, properties {bLooping=true, soundWaveAssetPtr=&/Ashton_Chicago_Throw_Loop_01}}
 {id=WavePlayer_1, properties {bLooping=true, soundWaveAssetPtr=&/Ashton_Chicago_Throw_Loop_02}}
]
properties {
 attenuationSettings=&Fort_Weapon_Projectile_Close_Athena_Att
 bOverrideConcurrency=true
 maxDistance=4950
 soundClassObject=&Fort_WeaponRanged
 volumeMultiplier=1
 attenuationOverrides {falloffDistance=5000, attenuationShapeExtents {x=1000}}
 concurrencyOverrides {resolutionRule=PreventNew}
}
